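# Masked, lit decal material for RL196 vehicle wheels, flagged for skeletal mesh and mesh deformer use.
# A static switch can mirror U for the shared decal, mask and normal texture UVs.
# The decal is tinted by BaseColor, then mask R, G and B successively lerp in Color_A, Color_B and Color_C.
# That result feeds an Emissive multiply and further decal multiplies, alongside scalar Metallic and Roughness parameters.
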Material /VehicleCosmetics/Wheels/RL196/Materials/M_Wheel_Decal_Simple_Extended
graph {
 properties {
  bAutomaticallySetUsageInEditor=false
  bCastRayTracedShadows=false
  bUsedWithMeshDeformer=true
  bUsedWithSkeletalMesh=true
  blendMode=BLEND_Masked
  usedShadingModels=Default Lit
 }
 nodes [
  n1 = TextureCoordinate()
  n2 = ComponentMask(None=@n1) {r=true}
  n3 = StaticSwitchParameter(True=AppendVector(A=OneMinus(None=@n2), B=ComponentMask(None=@n1) {g=true}), False=@n1) {group=Base, parameterName=Mirror Decal, sortPriority=-10}
  n4 = TextureSampleParameter2D(UVs=@n3) {group=Base, parameterName=Decal, sortPriority=-5, texture=/VehicleCosmetics/Textures/T_White}
  n5 = Multiply(A=@n4.RGB, B=VectorParameter() {parameterName=BaseColor, defaultValue=<rgba 0.5 0.5 0.5 1>}.RGB)
  n6 = VectorParameter() {parameterName=Color_A, defaultValue=<rgba 1 0 0 1>}
  n7 = TextureSampleParameter2D(UVs=@n3) {group=Base, parameterName=Mask, sortPriority=-3, texture=&T_White}
  n8 = LinearInterpolate(
   A=LinearInterpolate(
    A=LinearInterpolate(A=@n5, B=@n6.RGB, Alpha=@n7.R) {constA=1}
    B=VectorParameter() {parameterName=Color_B, defaultValue=<rgba 0 1 0 1>}.RGB
    Alpha=@n7.G
   )
   B=VectorParameter() {parameterName=Color_C, defaultValue=<rgba 0 0 1 1>}.RGB
   Alpha=@n7.B
  )
  Multiply(A=@n8, B=ScalarParameter() {group=Base, parameterName=Emissive, sortPriority=5})
  Multiply(A=Multiply(A=@n4.RGB, B=@n8), B=@n8)
  n9 = TextureSampleParameter2D(UVs=@n3) {group=Base, parameterName=Decal Normal, samplerType=SAMPLERTYPE_Normal, sortPriority=-4, texture=/VehicleCosmetics/Textures/T_BlankNormal_B}
  Reroute(None=@n9.RGB)
  ScalarParameter() {group=Base, parameterName=Metallic, sortPriority=3}
  ScalarParameter() {defaultValue=1, group=Base, parameterName=Roughness, sortPriority=4}
 ]
}
